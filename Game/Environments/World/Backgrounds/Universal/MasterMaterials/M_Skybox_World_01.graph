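# Master material for lit, fully rough background skybox terrain, with switchable diffuse, normal, overlay, foliage and snow layers.
# World-aligned blend and vertex color drive a height-blended tree canopy with sun-shifted shadows and optional bump offset; snow follows normal Z.
# Collection parameters drive time-of-day distance darkening and emissive haze with a lit region, faded by storm strength.
# Both haze terms end in lerps with a constant zero alpha, leaving them inactive.
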Material /Game/Environments/World/Backgrounds/Universal/MasterMaterials/M_Skybox_World_01
graph {
 properties {bFullyRough=true, bTangentSpaceNormal=false, pixelDepthOffsetMode=PDOM_Legacy, usedShadingModels=Default Lit}
 nodes [
  Constant() {r=65536}
  Constant() {r=-1}
  Constant() {r=4}
  Constant() {r=16}
  Constant()
  n1 = ComponentMask(None=WorldPosition() {worldPositionShaderOffset=WPT_ExcludeAllShaderOffsets}.XYZ) {g=true, r=true}
  n2 = CollectionParameter() {
   collection=&/MPC_Skybox_World
   parameterName="Morn_R=Distance_G=DiffuseAmt_B=EmissiveAmt"
  }
  n3 = CollectionParameter() {
   collection=&MPC_Skybox_World
   parameterName="Day_R=Distance_G=DiffuseAmt_B=EmissiveAmt"
  }
  n4 = CollectionParameter() {
   collection=&MPC_Skybox_World
   parameterName="Eve_R=Distance_G=DiffuseAmt_B=EmissiveAmt"
  }
  n5 = CollectionParameter() {
   collection=&MPC_Skybox_World
   parameterName="Night_R=Distance_G=DiffuseAmt_B=EmissiveAmt"
  }
  n6 = MaterialFunctionCall(Float3=MaterialFunctionCall(Color_Morning=@n2, Color_Day=@n3, Color_Evening=@n4, Color_Night=@n5) {materialFunction=&Color_Time_of_Day}.Result) {materialFunction=&BreakOutFloat3Components}
  n7 = SphereMask(A=@n1, B=Constant2Vector(), Radius=@n6.R, Hardness=Constant())
  n8 = VertexInterpolator(VS=LinearInterpolate(A=@n7, Alpha=@n6.G))
  n9 = TextureCoordinate()
  n10 = ScalarParameter() {group=Diffuse, parameterName=OverlayDiffuseUVOffset}
  n11 = TextureSampleParameter2D(UVs=Multiply(A=Add(A=@n9, B=@n10), B=ScalarParameter() {defaultValue=1, group=Diffuse, parameterName=OverlayDiffuseSize})) {
   group=Textures
   parameterName=OverlayUV0Texture
   texture=/Game/Environments/World/Backgrounds/Locales/Temperate/Textures/T_Skybox_Lowlands_Base2_D
  }
  n12 = TextureObjectParameter() {group=Textures, parameterName=WorldSpaceDiffuse, texture=/Game/Packages/DS_Fortnite_Terrain_NoLOD/Textures/T_Cave_Exterior_Wall_D}
  n13 = ScalarParameter() {defaultValue=1, group=Diffuse, parameterName=WorldDiffuseSize}
  n14 = Multiply(A=VectorParameter() {group=Diffuse, parameterName=WorldDiffuse3dSize, defaultValue=<rgba 1 1 0.5 1>}.RGB, B=@n13)
  n15 = Add(A=WorldPosition().XYZ, B=Multiply(A=@n13, B=ScalarParameter() {group=Diffuse, parameterName=WorldDiffuseUVOffset}))
  n16 = Multiply(A=TextureCoordinate(), B=ScalarParameter() {defaultValue=1, group=Diffuse, parameterName=DiffuseTile})
  n17 = StaticSwitchParameter(
   True=StaticSwitchParameter(
    True=MaterialFunctionCall(TextureObject=@n12, TextureSize=@n14, WorldPosition=@n15) {materialFunction=&WorldAlignedTexture}.XYZ Texture
    False=TextureSampleParameter2D(UVs=@n16) {group=Textures, parameterName=DiffuseTex, texture=&T_Cave_Exterior_Wall_D}.RGB
   ) {group=Diffuse, parameterName=UseWorldAlignedDiffuse}
   False=Constant() {r=1}
  ) {defaultValue=true, group=Diffuse, parameterName=UseDiffuseTexture}
  n18 = Multiply(A=VectorParameter() {desc=Color, group=Diffuse, parameterName=DiffuseColor1, defaultValue=<rgba 1 1 1 1>}.RGB, B=@n17)
  n19 = VectorParameter() {group=Diffuse, parameterName=DiffuseColor2, defaultValue=<rgba 0 1 0 1>}
  n20 = TextureSampleParameter2D(UVs=Multiply(A=TextureCoordinate(), B=ScalarParameter() {defaultValue=5, group=Normal, parameterName=Detail Normal Tiling})) {
   group=Textures
   parameterName=DetailNormal
   samplerType=SAMPLERTYPE_Normal
   texture=/Game/Packages/DS_Fortnite_Terrain_NoLOD/Textures/T_Cave_Exterior_Wall_N
  }
  n21 = TextureCoordinate()
  n22 = ScalarParameter() {group=Normal, parameterName=TangentNormalUVOffset}
  n23 = TextureSampleParameter2D(UVs=Multiply(A=Add(A=@n21, B=@n22), B=ScalarParameter() {defaultValue=1, group=Normal, parameterName=TangentNormalSize})) {
   group=Textures
   parameterName=TangentNormal
   samplerType=SAMPLERTYPE_Normal
   texture=&T_Cave_Exterior_Wall_N
  }
  n24 = StaticSwitchParameter(
   True=MaterialFunctionCall(BaseNormal=@n23.RGB, AdditionalNormal=@n20.RGB) {materialFunction=&BlendAngleCorrectedNormals}.Result
   False=@n23.RGB
  ) {group=Normal, parameterName=DetailNormalAffectsFoliage}
  n25 = TextureObjectParameter() {
   group=Textures
   parameterName=WorldSpaceNormal
   samplerType=SAMPLERTYPE_Normal
   texture=&T_Cave_Exterior_Wall_N
  }
  n26 = MaterialFunctionCall(TextureObject=@n25, TextureSize=ScalarParameter() {defaultValue=100000, group=Normal, parameterName=WorldNormalSize}) {materialFunction=&WorldAlignedNormal}
  n27 = Constant3Vector() {constant=<rgba 0 0 1 0>}
  n28 = MaterialFunctionCall!(
   In Explicit Normal=StaticSwitchParameter(True=StaticSwitchParameter(True=@n26.XYZ Texture, False=@n24) {group=Normal, parameterName=UseWorldAlignedNormal}, False=@n27) {defaultValue=true, group=Normal, parameterName=UseNormalsTexture}
   In World Vector
   Blend Sharpness=ScalarParameter() {defaultValue=4, group=FoliageMask, parameterName=WorldBlendSharpness}
   Blend Bias=ScalarParameter() {defaultValue=-1, group=FoliageMask, parameterName=WorldBlendBias}
   Clamped?=StaticBool() {value=true}
   Alpha
  ) {materialFunction=/Engine/Functions/Engine_MaterialFunctions01/AlphaBlend/WorldAlignedBlend}
  n29 = OneMinus(None=Power(Base=OneMinus(None=@n28.w/ Explicit Normal), Exp=ScalarParameter() {defaultValue=1.5, group=Diffuse, parameterName=WorldZMaskPower}))
  n30 = TextureSampleParameter2D(UVs=Multiply(A=TextureCoordinate(), B=ScalarParameter() {defaultValue=1, group=Diffuse, parameterName=MaskTile})) {
   group=Textures
   parameterName=Mask
   texture=/Game/Environments/World/Backgrounds/Locales/Temperate/Textures/T_Skybox_Lowlands_Ambient_occlusion
  }
  n31 = ScalarParameter() {defaultValue=-0.5, group=Diffuse, parameterName=MaskOffset}
  n32 = Clamp(None=Multiply(A=Add(A=@n30.R, B=@n31), B=ScalarParameter() {defaultValue=1, group=Diffuse, parameterName=MaskMult}))
  n33 = StaticSwitchParameter(
   True=LinearInterpolate(
    A=@n18
    B=Multiply(A=@n17, B=@n19.RGB)
    Alpha=StaticSwitchParameter(True=@n29, False=@n32) {group=Diffuse, parameterName=UseAlternateMaskType}
   )
   False=@n18
  ) {group=Diffuse, parameterName=UseMaskFor2ColorBlend}
  n34 = StaticSwitchParameter(True=Multiply(A=@n33, B=@n11.RGB), False=@n33) {group=Diffuse, parameterName=UseOverlayTexture}
  n35 = Constant() {r=1}
  n36 = Multiply(A=@n28.w/ Explicit Normal, B=VertexColor().G)
  n37 = LinearInterpolate(
   A=ScalarParameter() {defaultValue=-512, group=FoliageMask, parameterName=Blend Inner Min}
   B=ScalarParameter() {defaultValue=-6, group=FoliageMask, parameterName=BlendMin}
   Alpha=@n36
  )
  n38 = ScalarParameter() {defaultValue=16, group=FoliageMask, parameterName=BlendMax}
  n39 = CollectionParameter() {
   collection=&FortniteMaterialParameters
   parameterName=SunlightDirectionalVector
  }
  n40 = ComponentMask(None=Transform(None=ComponentMask(None=@n39) {b=true, g=true, r=true}) {transformSourceType=TRANSFORMSOURCE_World, transformType=TRANSFORM_Tangent}) {g=true, r=true}
  n41 = ScalarParameter() {defaultValue=0.015, group=FoliageShadow, parameterName=ShadowShift}
  n42 = Multiply(A=TextureCoordinate(), B=ScalarParameter() {defaultValue=4, group=Foliage, parameterName=FoliageTiling})
  n43 = TextureSample(UVs=Add(A=Multiply(A=@n40, B=@n41), B=@n42)) {samplerType=SAMPLERTYPE_LinearGrayscale, texture=/Game/Environments/World/Backgrounds/Locales/Temperate/Textures/T_Skybox_TreeCanopy_01_A}
  n44 = Add(A=LinearInterpolate(A=@n37, B=@n38, Alpha=@n43.R), B=@n36) {desc=Shifted Shadow}
  n45 = Add(
   A=Divide(A=OneMinus(None=@n44), B=ScalarParameter() {defaultValue=12, group=FoliageShadow, parameterName=ShadowDivisor})
   B=ScalarParameter() {defaultValue=0.1, group=FoliageShadow, parameterName=ShadowSize}
  )
  n46 = OneMinus(None=VertexColor().G)
  n47 = LinearInterpolate(A=ScalarParameter() {defaultValue=0.2, group=FoliageShadow, parameterName=FoliageShadowAmount}, Alpha=Clamp(None=Add(A=@n45, B=@n46)))
  n48 = StaticSwitchParameter(
   True=ShadingPathSwitch(Default=QualitySwitch(Default=@n35, Low=@n35, High=@n47), Mobile=@n35) {desc=Migrated from FeatureLevelSwitch}
   False=@n35
  ) {group=Foliage, parameterName=UseFoliage}
  n49 = TextureSample(UVs=@n42) {samplerType=SAMPLERTYPE_LinearGrayscale, texture=&T_Skybox_TreeCanopy_01_A}
  n50 = BumpOffset(
   Coordinate=@n42
   Height=Power(Base=@n49.R, Exp=ScalarParameter() {defaultValue=10, group=Foliage, parameterName=BumpOffsetHeightPower}) {constExponent=10}
   HeightRatioInput=ScalarParameter() {defaultValue=0.05, group=Foliage, parameterName=BumpOffsetRatio}
  ) {referencePlane=0}
  n51 = ShadingPathSwitch(
   Default=QualitySwitch(Default=@n42, Low=@n42, High=StaticSwitchParameter(True=@n50, False=@n42) {group=Foliage, parameterName=UseFoliageBumpOffset})
   Mobile=@n42
  ) {desc=Migrated from FeatureLevelSwitch}
  n52 = TextureSampleParameter2D(UVs=@n51) {group=Textures, parameterName=FoliageDiffuse, texture=/Game/Environments/World/Backgrounds/Locales/Temperate/Textures/T_Skybox_TreeCanopy_03}
  n53 = VectorParameter() {group=Foliage, parameterName=FoliageColor, defaultValue=<rgba 0.4428 0.67 0.2023 1>}
  n54 = Multiply(A=@n37)
  n55 = Clamp(None=Add(A=@n36, B=LinearInterpolate(A=@n54, B=@n38, Alpha=@n49.R)))
  n56 = StaticSwitchParameter(True=LinearInterpolate(A=Multiply(A=@n34, B=@n48), B=Multiply(A=@n52.RGB, B=@n53.RGB), Alpha=@n55), False=@n34) {group=Foliage, parameterName=UseFoliage}
  n57 = VectorParameter() {group=Snow, parameterName=SnowColor, defaultValue=<rgba 3 3 3 1>}
  n58 = ScalarParameter() {defaultValue=2, group=Snow, parameterName=SnowWorldZMaskMod}
  n59 = VectorParameter() {group=Foliage, parameterName=FoliageNormalIntensity, defaultValue=<rgba 2 2 1 0>}
  n60 = TextureSampleParameter2D(UVs=@n51) {
   group=Textures
   parameterName=FoliageNormal
   samplerType=SAMPLERTYPE_Normal
   texture=/Game/Environments/World/Backgrounds/Locales/Temperate/Textures/T_Skybox_TreeCanopy_01_N
  }
  n61 = Multiply(A=@n23.RGB, B=VectorParameter() {group=Normal, parameterName=TangentNormalIntensity, defaultValue=<rgba 1 1 1 0>}.RGB)
  n62 = MaterialFunctionCall(
   BaseNormal=@n61
   AdditionalNormal=Multiply(A=@n20.RGB, B=VectorParameter() {group=Normal, parameterName=DetailNormalStrength, defaultValue=<rgba 0.3 0.3 1 1>}.RGB)
  ) {materialFunction=&BlendAngleCorrectedNormals}
  n63 = StaticSwitchParameter(
   True=StaticSwitchParameter(
    True=@n26.XYZ Texture
    False=StaticSwitchParameter(True=@n62.Result, False=@n61) {defaultValue=true, group=Normal, parameterName=UseDetailNormal}
   ) {group=Normal, parameterName=UseWorldAlignedNormal}
   False=@n27
  ) {defaultValue=true, group=Normal, parameterName=UseNormalsTexture}
  n64 = StaticSwitchParameter(True=@n63, False=Transform(None=@n63)) {group=Normal, parameterName=UseWorldAlignedNormal}
  n65 = LinearInterpolate(A=@n64, B=Transform(None=Multiply(A=@n59.RGB, B=@n60.RGB)), Alpha=@n55)
  n66 = StaticSwitchParameter(True=@n65, False=@n64) {group=Foliage, parameterName=UseFoliage}
  n67 = ComponentMask(None=Normalize(VectorInput=@n66)) {b=true}
  n68 = Saturate(None=Add(A=ScalarParameter() {defaultValue=-1.2, group=Snow, parameterName=SnowWorldZMaskAdd}, B=Multiply(A=@n58, B=@n67)))
  n69 = Saturate(None=StaticSwitchParameter(True=LinearInterpolate(A=@n56, B=@n57.RGB, Alpha=@n68), False=@n56) {group=Snow, parameterName=UseSnow})
  n70 = Constant()
  LinearInterpolate(A=@n69, B=Multiply(A=@n69, B=ComponentMask(None=@n8.PS) {r=true}) {desc=BaseColor}, Alpha=@n70) {constB=0}
  n71 = Clamp(None=Multiply(
    A=Constant() {r=3}
    B=CollectionParameter() {collection=&FortniteMaterialParameters, parameterName=StormStrength}
   ))
  n72 = Multiply(A=Multiply(A=OneMinus(None=@n71), B=@n6.B), B=OneMinus(None=@n7))
  n73 = CollectionParameter() {collection=&MPC_Skybox_World, parameterName=MorningColor}
  n74 = CollectionParameter() {collection=&MPC_Skybox_World, parameterName=DayColor}
  n75 = CollectionParameter() {collection=&MPC_Skybox_World, parameterName=EveningColor}
  n76 = CollectionParameter() {collection=&MPC_Skybox_World, parameterName=NightColor}
  n77 = CollectionParameter() {collection=&MPC_Skybox_World, parameterName=MorningLitColor}
  n78 = CollectionParameter() {collection=&MPC_Skybox_World, parameterName=DayLitColor}
  n79 = CollectionParameter() {collection=&MPC_Skybox_World, parameterName=EveningLitColor}
  n80 = CollectionParameter() {collection=&MPC_Skybox_World, parameterName=NightLitColor}
  n81 = Add(A=@n1, B=Constant() {r=4000000})
  n82 = CollectionParameter() {
   collection=&MPC_Skybox_World
   parameterName="Morn_R=LitLength_G=LitRot_B=LitOffset"
  }
  n83 = CollectionParameter() {
   collection=&MPC_Skybox_World
   parameterName="Day_R=LitLength_G=LitRot_B=LitOffset"
  }
  n84 = CollectionParameter() {
   collection=&MPC_Skybox_World
   parameterName="Eve_R=LitLength_G=LitRot_B=LitOffset"
  }
  n85 = CollectionParameter() {
   collection=&MPC_Skybox_World
   parameterName="Night_R=LitLength_G=LitRot_B=LitOffset"
  }
  n86 = MaterialFunctionCall(Float3=MaterialFunctionCall(Color_Morning=@n82, Color_Day=@n83, Color_Evening=@n84, Color_Night=@n85) {materialFunction=&Color_Time_of_Day}.Result) {materialFunction=&BreakOutFloat3Components}
  n87 = AppendVector(A=@n86.G, B=@n86.G)
  n88 = LinearInterpolate(
   A=ComponentMask(None=MaterialFunctionCall(Color_Morning=@n73, Color_Day=@n74, Color_Evening=@n75, Color_Night=@n76) {materialFunction=&Color_Time_of_Day}.Result) {b=true, g=true, r=true}
   B=ComponentMask(None=MaterialFunctionCall(Color_Morning=@n77, Color_Day=@n78, Color_Evening=@n79, Color_Night=@n80) {materialFunction=&Color_Time_of_Day}.Result) {b=true, g=true, r=true}
   Alpha=SphereMask(A=Subtract(A=@n81, B=@n87), B=Constant2Vector(), Radius=@n86.R, Hardness=Constant())
  )
  LinearInterpolate(B=VertexInterpolator(VS=Multiply(A=@n72, B=@n88) {desc=Emissive}).PS, Alpha=@n70) {constB=0}
  Multiply(
   A=OneMinus(None=VertexColor().B)
   B=Multiply(A=Constant3Vector() {constant=<rgba 0 0 1 0>}, B=ScalarParameter() {defaultValue=-8192, group=WPO, parameterName=Painted Height})
  )
  Multiply(A=@n37)
  Reroute(None=@n66)
  ScalarParameter() {defaultValue=0.1, parameterName=Spec}
  ScalarParameter() {defaultValue=0.85, parameterName=Roughness}
  StaticSwitchParameter(True=@n65, False=@n64) {group=Foliage, parameterName=UseFoliageNormalsTexture}
  VectorParameter() {parameterName=SkyboxFogColor, defaultValue=<rgba 0.2484 0.2991 0.92 1>}
  VectorParameter() {parameterName=SkyboxFogLightColor, defaultValue=<rgba 0.3049 0.5404 1.5 1>}
 ]
}
